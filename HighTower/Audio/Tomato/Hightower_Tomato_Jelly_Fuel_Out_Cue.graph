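# Sound cue for the Tomato jelly item's fuel-out sound, by its name a jetpack-style gadget running out of fuel.
# A single wave player reuses the shared Athena jetpack fuel-out sound, fed through a modulator that randomly raises its pitch slightly.
# It plays at reduced volume under the gadget sound class with the jetpack attenuation settings.
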SoundCue /HighTower/Audio/Tomato/Hightower_Tomato_Jelly_Fuel_Out_Cue
root=Modulator_0
nodes [
 {id=Modulator_0, children [WavePlayer_0], properties {pitchMax=1.1, pitchMin=1}}
 {id=WavePlayer_0, properties {soundWaveAssetPtr=/FNGameplayCues/Athena/Sounds/Items/Jetpack/Jetpack_FuelOut_01}}
]
properties {
 attenuationSettings=/Game/Sounds/Attenuation/Fort_Jetpack_Att
 maxDistance=13067
 soundClassObject=&Fort_Gadgets
 volumeMultiplier=0.6
}
