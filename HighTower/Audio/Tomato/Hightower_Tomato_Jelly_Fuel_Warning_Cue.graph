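# Sound cue for the Hightower Tomato jelly fuel warning, apparently a low-fuel alert for a jetpack-style gadget.
# A single wave player feeds the output directly with one warning sample, with no randomization, looping or modulation.
# The cue plays at reduced volume under the gadgets sound class and uses the shared jetpack attenuation asset for distance falloff.
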SoundCue /HighTower/Audio/Tomato/Hightower_Tomato_Jelly_Fuel_Warning_Cue
root=WavePlayer_2
nodes [{id=WavePlayer_2, properties {soundWaveAssetPtr=&/Hightower_Tomato_Jelly_Fuel_Warning_01}}]
properties {
 attenuationSettings=/Game/Sounds/Attenuation/Fort_Jetpack_Att
 maxDistance=13067
 soundClassObject=&Fort_Gadgets
 volumeMultiplier=0.5
}
