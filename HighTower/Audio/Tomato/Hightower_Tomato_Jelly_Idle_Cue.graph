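# Sound cue for the idle loop of the Hightower Tomato Jelly, reusing jetpack idle loop audio.
# A quality level node selects between two looping wave players: the first quality slot plays the standard jetpack loop, and the remaining three share the variant suffixed "opt".
# The cue plays at reduced volume in the gadgets sound class, using the shared jetpack attenuation settings.
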SoundCue /HighTower/Audio/Tomato/Hightower_Tomato_Jelly_Idle_Cue
root=QualityLevel_0
nodes [
 {id=QualityLevel_0, children [WavePlayer_0,WavePlayer_1,WavePlayer_1,WavePlayer_1]}
 {id=WavePlayer_0, properties {bLooping=true, soundWaveAssetPtr=/FNGameplayCues/Athena/Sounds/Items/Jetpack/Jetpack_IdleLoop_01}}
 {id=WavePlayer_1, properties {bLooping=true, soundWaveAssetPtr=/FNGameplayCues/Athena/Sounds/Items/Jetpack/Jetpack_IdleLoop_01_opt}}
]
properties {
 attenuationSettings=/Game/Sounds/Attenuation/Fort_Jetpack_Att
 maxDistance=13067
 soundClassObject=&Fort_Gadgets
 volumeMultiplier=0.65
}
