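# Sound cue for the start of the Tomato Jelly thrust in HighTower, played as a thrust-on one-shot.
# A Random node picks one of three equally weighted thrust-on wave variations and feeds it through a Modulator at the root.
# It uses the shared jetpack attenuation and the gadgets sound class, with a reduced volume multiplier.
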SoundCue /HighTower/Audio/Tomato/Hightower_Tomato_Jelly_Start_Cue
root=Modulator_0
nodes [
 {id=Modulator_0, children [Random_0]}
 {id=Random_0, children [WavePlayer_0,WavePlayer_2,WavePlayer_3], properties {weights [1,1,1]}}
 {id=WavePlayer_0, properties {soundWaveAssetPtr=&/Hightower_Tomato_Jelly_Thrust_On_01}}
 {id=WavePlayer_2, properties {soundWaveAssetPtr=&/Hightower_Tomato_Jelly_Thrust_On_02}}
 {id=WavePlayer_3, properties {soundWaveAssetPtr=&/Hightower_Tomato_Jelly_Thrust_On_03}}
]
properties {
 attenuationSettings=/Game/Sounds/Attenuation/Fort_Jetpack_Att
 maxDistance=13067
 soundClassObject=&Fort_Gadgets
 volumeMultiplier=0.6
}
