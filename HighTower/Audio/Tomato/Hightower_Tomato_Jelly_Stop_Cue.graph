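# Sound cue for the Hightower Tomato Jelly thrust-off sound, played when the jelly thrust stops.
# A Random node picks one of three equally weighted thrust-off waves and feeds a Modulator that varies pitch slightly downward.
# Output is at reduced volume in the gadgets sound class, spatialized with the shared jetpack attenuation.
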SoundCue /HighTower/Audio/Tomato/Hightower_Tomato_Jelly_Stop_Cue
root=Modulator_0
nodes [
 {id=Modulator_0, children [Random_0], properties {pitchMax=1, pitchMin=0.9}}
 {id=Random_0, children [WavePlayer_1,WavePlayer_2,WavePlayer_3], properties {weights [1,1,1]}}
 {id=WavePlayer_1, properties {soundWaveAssetPtr=&/Hightower_Tomato_Jelly_Thrust_Off_01}}
 {id=WavePlayer_2, properties {soundWaveAssetPtr=&/Hightower_Tomato_Jelly_Thrust_Off_02}}
 {id=WavePlayer_3, properties {soundWaveAssetPtr=&/Hightower_Tomato_Jelly_Thrust_Off_03}}
]
properties {
 attenuationSettings=/Game/Sounds/Attenuation/Fort_Jetpack_Att
 maxDistance=13067
 soundClassObject=&Fort_Gadgets
 volumeMultiplier=0.6
}
