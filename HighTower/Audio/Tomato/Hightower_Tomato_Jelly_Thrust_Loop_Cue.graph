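# Sound cue that plays the looping Tomato jelly thrust sound in HighTower.
# A single looping wave player feeds a Modulator left at its default settings, which is the root node.
# It routes to the gadgets sound class at reduced volume and borrows the shared jetpack attenuation asset for distance falloff.
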SoundCue /HighTower/Audio/Tomato/Hightower_Tomato_Jelly_Thrust_Loop_Cue
root=Modulator_0
nodes [
 {id=Modulator_0, children [WavePlayer_2]}
 {id=WavePlayer_2, properties {bLooping=true, soundWaveAssetPtr=&/Hightower_Tomato_Jelly_Thrust_Loop_01}}
]
properties {
 attenuationSettings=/Game/Sounds/Attenuation/Fort_Jetpack_Att
 maxDistance=13067
 soundClassObject=&Fort_Gadgets
 volumeMultiplier=0.7
}
