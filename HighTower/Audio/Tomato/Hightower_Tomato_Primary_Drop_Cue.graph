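# Sound cue for the drop sound of the HighTower Tomato primary item.
# A single wave player plays one drop wave at slightly reduced volume, with no randomization or mixing.
# It uses the shared weapon-reload attenuation asset and the loot-collect sound class.
# Its own concurrency override prevents new instances from starting once the limit is reached, rather than stopping playing ones.
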SoundCue /HighTower/Audio/Tomato/Hightower_Tomato_Primary_Drop_Cue
root=WavePlayer_0
nodes [{id=WavePlayer_0, properties {soundWaveAssetPtr=&/Hightower_Tomato_Item_Drop_01}}]
properties {
 attenuationSettings=&Fort_Weapon_Reload_Att
 bOverrideConcurrency=true
 maxDistance=2250
 soundClassObject=&Fort_Loot_Collect
 volumeMultiplier=0.8413
 attenuationOverrides {distanceAlgorithm=NaturalSound, falloffDistance=4000, attenuationShapeExtents {x=1000}}
 concurrencyOverrides {resolutionRule=PreventNew}
}
