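# Sound cue for the equip foley of the Hightower Tomato primary weapon.
# A single wave player feeds one equip sound straight to the output, with no randomization or modulation.
# It uses the shared weapon reload attenuation and the weapon foley sound class.
# A concurrency override blocks new instances from starting once the limit is reached.
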SoundCue /HighTower/Audio/Tomato/Hightower_Tomato_Primary_Equip_Cue
root=WavePlayer_1
nodes [{id=WavePlayer_1, properties {soundWaveAssetPtr=&/Hightower_Tomato_Primary_Equip_01}}]
properties {
 attenuationSettings=&Fort_Weapon_Reload_Att
 bOverrideConcurrency=true
 maxDistance=2250
 soundClassObject=&Fort_WeaponFoley
 volumeMultiplier=1
 attenuationOverrides {falloffDistance=4100, attenuationShapeExtents {x=900}}
 concurrencyOverrides {resolutionRule=PreventNew}
}
