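# Sound cue for the second stage of the Hightower Tomato primary jump, filed under the HighTower audio content.
# A single wave player feeds a modulator at the root, which is left at its usual settings.
# The cue plays in the Fort_Gadgets sound class at slightly reduced volume, with shared ability status-effect activation attenuation.
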SoundCue /HighTower/Audio/Tomato/Hightower_Tomato_Primary_Jump_Stage_02_Cue
root=Modulator_0
nodes [
 {id=Modulator_0, children [WavePlayer_0]}
 {id=WavePlayer_0, properties {soundWaveAssetPtr=&/Hightower_Tomato_Primary_Jump_Stage_02_01}}
]
properties {
 attenuationSettings=/Game/Sounds/Attenuation/Fort_Ability_Status_Effect_Activation_Att
 maxDistance=6000
 soundClassObject=&Fort_Gadgets
 volumeMultiplier=0.8
}
