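# Sound cue for the HighTower Tomato primary lock-on, playing a target indicator sound.
# A single wave player feeds the output directly, with no randomization or modulation.
# It uses the UI sound class with gadget-activate attenuation.
# A concurrency override allows one instance at a time, stopping the oldest so a new lock-on replaces the previous sound.
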SoundCue /HighTower/Audio/Tomato/Hightower_Tomato_Primary_Lockon_Cue
root=WavePlayer_1
nodes [{id=WavePlayer_1, properties {soundWaveAssetPtr=&/Hightower_Tomato_Primary_Target_Indicator_01}}]
properties {
 attenuationSettings=&Fort_GadgetActivate_Att
 bOverrideConcurrency=true
 maxDistance=4000
 soundClassObject=&Fort_UI
 volumeMultiplier=1
 concurrencyOverrides {maxCount=1, resolutionRule=StopOldest}
}
