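# Sound cue for the whoosh-by of the Hightower Tomato primary projectile, classed as a ranged weapon sound.
# A random node picks one of five equally weighted whoosh recordings and feeds it through a modulator that varies pitch.
# Attenuation is overridden to disable distance falloff and set a wide stereo spread.
# Playback runs at reduced volume and is limited by the shared Athena bullet whiz-by concurrency rule.
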SoundCue /HighTower/Audio/Tomato/Hightower_Tomato_Primary_Proj_By_Cue
root=Modulator_0
nodes [
 {id=Modulator_0, children [Random_1], properties {pitchMax=1.15}}
 {id=Random_1, children [WavePlayer_5, WavePlayer_7, WavePlayer_8, WavePlayer_9, WavePlayer_10], properties {weights [1,1,1,1,1]}}
 {id=WavePlayer_10, properties {soundWaveAssetPtr=&/Hightower_Tomato_Primary_Proj_Whoosh_By_06}}
 {id=WavePlayer_5, properties {soundWaveAssetPtr=&/Hightower_Tomato_Primary_Proj_Whoosh_By_01}}
 {id=WavePlayer_7, properties {soundWaveAssetPtr=&/Hightower_Tomato_Primary_Proj_Whoosh_By_02}}
 {id=WavePlayer_8, properties {soundWaveAssetPtr=&/Hightower_Tomato_Primary_Proj_Whoosh_By_03}}
 {id=WavePlayer_9, properties {soundWaveAssetPtr=&/Hightower_Tomato_Primary_Proj_Whoosh_By_04}}
]
properties {
 bOverrideAttenuation=true
 maxDistance=2097152
 soundClassObject=&Fort_WeaponRanged
 volumeMultiplier=0.6
 attenuationOverrides {bAttenuate=false, stereoSpread=5000}
 concurrencySet [/Game/Sounds/Concurrency_Rules/Athena_BulletWhizBy_Conc]
}
