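# Sound cue for the explosion of the Hightower Tomato primary projectile.
# A mixer blends close, medium and distant layers, each with its own explosion attenuation, so the sound changes with listener distance.
# Each layer picks one of four waves at random and varies its pitch; the close layer also varies volume and starts after a short fixed delay.
# A concurrency override caps it at two instances, stopping the oldest.
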SoundCue /HighTower/Audio/Tomato/Hightower_Tomato_Primary_Proj_Expl_Cue
root=Mixer_0
nodes [
 {id=Attenuation_0, children [Modulator_2], properties {attenuationSettings=/Game/Sounds/Attenuation/Fort_Explosion_Medium_Athena_Att}}
 {id=Attenuation_1, children [Modulator_0], properties {attenuationSettings=/Game/Sounds/Attenuation/Fort_Explosion_Distant_Athena_Att}}
 {
  id=Attenuation_7
  children [Modulator_6]
  properties {attenuationSettings=/Game/Sounds/Attenuation/Fort_Explosion_Close_Athena_Att, attenuationOverrides {falloffDistance=1600}}
 }
 {id=Delay_0, children [Random_1], properties {delayMax=0.066, delayMin=0.066}}
 {id=Mixer_0, children [Attenuation_7, Attenuation_0, Attenuation_1], properties {inputVolume [1,1,1]}}
 {id=Modulator_0, children [Random_4], properties {pitchMax=1.1, pitchMin=0.9}}
 {id=Modulator_2, children [Random_2], properties {pitchMax=1.1, pitchMin=0.8}}
 {id=Modulator_6, children [Delay_0], properties {pitchMax=1.3, pitchMin=1, volumeMin=0.5}}
 {id=Random_1, children [WavePlayer_4,WavePlayer_5,WavePlayer_6,WavePlayer_7], properties {weights [1,1,1,1]}}
 {id=Random_2, children [WavePlayer_8, WavePlayer_9, WavePlayer_10, WavePlayer_11], properties {weights [1,1,1,1]}}
 {id=Random_4, children [WavePlayer_12, WavePlayer_13, WavePlayer_22, WavePlayer_23], properties {weights [1,1,1,1]}}
 {id=WavePlayer_10, properties {soundWaveAssetPtr=&/Hightower_Tomato_Primary_Proj_Expl_Med_03}}
 {id=WavePlayer_11, properties {soundWaveAssetPtr=&/Hightower_Tomato_Primary_Proj_Expl_Med_04}}
 {id=WavePlayer_12, properties {soundWaveAssetPtr=&/Hightower_Tomato_Primary_Proj_Expl_Dist_04}}
 {id=WavePlayer_13, properties {soundWaveAssetPtr=&/Hightower_Tomato_Primary_Proj_Expl_Dist_01}}
 {id=WavePlayer_22, properties {soundWaveAssetPtr=&/Hightower_Tomato_Primary_Proj_Expl_Dist_02}}
 {id=WavePlayer_23, properties {soundWaveAssetPtr=&/Hightower_Tomato_Primary_Proj_Expl_Dist_03}}
 {id=WavePlayer_4, properties {soundWaveAssetPtr=&/Hightower_Tomato_Primary_Proj_Expl_Close_01}}
 {id=WavePlayer_5, properties {soundWaveAssetPtr=&/Hightower_Tomato_Primary_Proj_Expl_Close_02}}
 {id=WavePlayer_6, properties {soundWaveAssetPtr=&/Hightower_Tomato_Primary_Proj_Expl_Close_03}}
 {id=WavePlayer_7, properties {soundWaveAssetPtr=&/Hightower_Tomato_Primary_Proj_Expl_Close_04}}
 {id=WavePlayer_8, properties {soundWaveAssetPtr=&/Hightower_Tomato_Primary_Proj_Expl_Med_01}}
 {id=WavePlayer_9, properties {soundWaveAssetPtr=&/Hightower_Tomato_Primary_Proj_Expl_Med_02}}
]
properties {
 bOverrideConcurrency=true
 maxDistance=27500
 soundClassObject=&Fort_WeaponRanged_Fire_3P_PSM
 volumeMultiplier=0.6
 concurrencyOverrides {maxCount=2, resolutionRule=StopOldest}
}
