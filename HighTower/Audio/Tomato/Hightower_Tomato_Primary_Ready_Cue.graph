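# Sound cue for the Hightower Tomato primary's ready sound, built from cooldown-end waves.
# A mixer layers two separately delayed waves, with the arms layer starting slightly before the main cooldown-end wave, both under half a second or so in.
# It uses the weapon foley sound class and weapon reload attenuation asset, and its own concurrency override prevents new instances once the limit is reached.
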SoundCue /HighTower/Audio/Tomato/Hightower_Tomato_Primary_Ready_Cue
root=Mixer_0
nodes [
 {id=Delay_0, children [WavePlayer_0], properties {delayMax=0.5, delayMin=0.5}}
 {id=Delay_1, children [WavePlayer_1], properties {delayMax=0.4, delayMin=0.4}}
 {id=Mixer_0, children [Delay_0,Delay_1], properties {inputVolume [1,1]}}
 {id=WavePlayer_0, properties {soundWaveAssetPtr=&/Hightower_Tomato_Primary_Cooldown_End_01}}
 {id=WavePlayer_1, properties {soundWaveAssetPtr=&/Hightower_Tomato_Primary_Cooldown_Arms_01}}
]
properties {
 attenuationSettings=&Fort_Weapon_Reload_Att
 bOverrideConcurrency=true
 maxDistance=2250
 soundClassObject=&Fort_WeaponFoley
 volumeMultiplier=1
 attenuationOverrides {falloffDistance=4100, attenuationShapeExtents {x=900}}
 concurrencyOverrides {resolutionRule=PreventNew}
}
