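# Sound cue for the Hightower Tomato secondary charge sound, routed to the third-person weapon events sound class.
# A single wave player feeds a default Modulator at the root, with reduced cue volume and a dedicated attenuation asset.
# A concurrency override limits it to two instances per owner, stopping the oldest when the limit is exceeded.
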SoundCue /HighTower/Audio/Tomato/Hightower_Tomato_Secondary_Charge_Cue
root=Modulator_0
nodes [
 {id=Modulator_0, children [WavePlayer_1]}
 {id=WavePlayer_1, properties {soundWaveAssetPtr=&/Hightower_Tomato_Secondary_Charge_01}}
]
properties {
 attenuationSettings=&/Hightower_Tomato_Secondary_Charge_Att
 bOverrideConcurrency=true
 maxDistance=15000
 soundClassObject=/Game/Sounds/FortSoundClassesAndModes/Fort_Events_Weapons_3P
 volumeMultiplier=0.6
 concurrencyOverrides {bLimitToOwner=true, maxCount=2, resolutionRule=StopOldest}
}
