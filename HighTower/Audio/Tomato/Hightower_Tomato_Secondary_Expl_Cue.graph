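# Sound cue for the Tomato secondary explosion in HighTower, mixing close, medium and distant layers, each with random pitch and its own attenuation node.
# The close layer is briefly delayed, varies in volume and overrides its falloff distance, while the other two are mixed louder.
# The distant layer uses an attenuation asset named for close explosions, not a distant one.
# Concurrency allows one instance per owner and rejects new ones, under a third-person ranged-weapon fire sound class.
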SoundCue /HighTower/Audio/Tomato/Hightower_Tomato_Secondary_Expl_Cue
root=Mixer_0
nodes [
 {id=Attenuation_0, children [Modulator_2], properties {attenuationSettings=&/Hightower_Tomato_Medium_Explosion_Att}}
 {id=Attenuation_1, children [Modulator_0], properties {attenuationSettings=&/Hightower_Tomato_Close_Explosion_Att}}
 {
  id=Attenuation_7
  children [Modulator_6]
  properties {attenuationSettings=&Hightower_Tomato_Close_Explosion_Att, attenuationOverrides {falloffDistance=1600}}
 }
 {id=Delay_0, children [WavePlayer_0], properties {delayMax=0.066, delayMin=0.066}}
 {id=Mixer_0, children [Attenuation_7, Attenuation_0, Attenuation_1], properties {inputVolume [1,2,2]}}
 {id=Modulator_0, children [WavePlayer_2], properties {pitchMax=1.1, pitchMin=0.9}}
 {id=Modulator_2, children [WavePlayer_1], properties {pitchMax=1.1, pitchMin=0.9}}
 {id=Modulator_6, children [Delay_0], properties {pitchMax=1.3, pitchMin=1, volumeMin=0.5}}
 {id=WavePlayer_0, properties {soundWaveAssetPtr=&/Hightower_Tomato_Secondary_Expl_Close_01}}
 {id=WavePlayer_1, properties {soundWaveAssetPtr=&/Hightower_Tomato_Secondary_Expl_Med_01}}
 {id=WavePlayer_2, properties {soundWaveAssetPtr=&/Hightower_Tomato_Secondary_Expl_Dist_01}}
]
properties {
 bOverrideConcurrency=true
 maxDistance=7500
 soundClassObject=&Fort_WeaponRanged_Fire_3P_PSM
 volumeMultiplier=1
 concurrencyOverrides {bLimitToOwner=true, maxCount=1, resolutionRule=PreventNew}
}
