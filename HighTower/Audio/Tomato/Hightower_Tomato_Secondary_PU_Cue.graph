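# Sound cue for the HighTower Tomato secondary item, apparently its pickup sound given the "PU" name and weapon pickup attenuation.
# A single wave player feeds a Modulator node at the root, playing one wave at reduced volume.
# It routes through the gadgets sound class and overrides concurrency so new instances are blocked once the limit is reached.
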SoundCue /HighTower/Audio/Tomato/Hightower_Tomato_Secondary_PU_Cue
root=Modulator_0
nodes [
 {id=Modulator_0, children [WavePlayer_0]}
 {id=WavePlayer_0, properties {soundWaveAssetPtr=&/Hightower_Tomato_Secondary_PU_01}}
]
properties {
 attenuationSettings=&Fort_Weapon_Pickup_Att
 bOverrideConcurrency=true
 maxDistance=3500
 soundClassObject=&Fort_Gadgets
 volumeMultiplier=0.6
 concurrencyOverrides {resolutionRule=PreventNew}
}
