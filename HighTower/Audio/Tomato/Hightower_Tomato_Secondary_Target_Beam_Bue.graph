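# Sound cue that plays the Hightower Tomato secondary target beam sound at half volume.
# A single wave player runs through a short fixed delay, then a SoundClass node that overrides the class to third-person ranged weapon fire.
# It borrows the secondary whoosh-by attenuation, and its own concurrency limits it to two instances per owner, stopping the oldest.
# The cue's debug flag is on.
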SoundCue /HighTower/Audio/Tomato/Hightower_Tomato_Secondary_Target_Beam_Bue
root=SoundClass_5
nodes [
 {id=Delay_0, children [WavePlayer_1], properties {delayMax=0.1, delayMin=0.1}}
 {id=SoundClass_5, children [Delay_0], properties {soundClassOverride=&Fort_WeaponRanged_Fire_3P_PSM}}
 {id=WavePlayer_1, properties {soundWaveAssetPtr=&/Hightower_Tomato_Secondary_Target_Beam_02}}
]
properties {
 attenuationSettings=&/Hightower_Tomato_Secondary_Whoosh_By_Att
 bDebug=true
 bOverrideConcurrency=true
 maxDistance=2048
 volumeMultiplier=0.5
 concurrencyOverrides {bLimitToOwner=true, maxCount=2, resolutionRule=StopOldest}
}
