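# Sound cue for the whoosh-by of the Hightower Tomato secondary, classed as a ranged weapon projectile effect.
# A single wave player feeds a modulator that varies pitch, with the cue's volume boosted.
# It uses its own attenuation asset and shares the Athena bullet whiz-by concurrency rule.
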SoundCue /HighTower/Audio/Tomato/Hightower_Tomato_Secondary_Whoosh_By_Cue
root=Modulator_0
nodes [
 {id=Modulator_0, children [WavePlayer_0], properties {pitchMax=1.15}}
 {id=WavePlayer_0, properties {soundWaveAssetPtr=&/Hightower_Tomato_Secondary_Whoosh_By_01}}
]
properties {
 attenuationSettings=&/Hightower_Tomato_Secondary_Whoosh_By_Att
 maxDistance=2048
 soundClassObject=&Fort_WeaponRanged_Projectile_Effects
 volumeMultiplier=2
 attenuationOverrides {bAttenuate=false, stereoSpread=5000}
 concurrencySet [/Game/Sounds/Concurrency_Rules/Athena_BulletWhizBy_Conc]
}
