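# MetaSound Source preset for the first-person primary fire sound of the HighTower Tomato weapon, built on MSS_Base_Weapon_Fire_1p_Simple.
# The parent supplies the graph; this preset feeds it six close-fire waves with pitch and volume randomization ranges, leaving the low-ammo sound list empty.
# Fort Character and Weapon interfaces expose gameplay state such as ammo and aiming.
# Output is one-shot 5.1, assigned to a first-person ranged-fire sound class and local gunfire concurrency.
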MetaSoundSource /HighTower/Audio/Tomato/MSS_Hightower_Tomato_Primary_Fire_1p_Preset
preset_parent=/FortMetaSoundWeapon/Items/Metasounds/MSS_Base_Weapon_Fire_1p_Simple
inputs [
 bAutoFire {type=Bool}
 Character.Affiliation {type=Enum:FortAudioAffiliation}
 Character.Health {type=Float}
 Character.UniqueID {type=Int32}
 LowAmmo.ADSMainVolumeMultiply {type=Float}
 LowAmmo.Random.MaxPitch {type=Float}
 LowAmmo.Random.MaxVolume {type=Float}
 LowAmmo.Random.MinPitch {type=Float}
 LowAmmo.Random.MinVolume {type=Float}
 LowAmmo.Sounds {type=WaveAsset:Array}
 LowAmmo.Threshold {type=Int32}
 MainFire.ADSMainVolumeMultiply {type=Float}
 MainFire.LFVolume {type=Float}
 MainFire.Random.MaxPitch {type=Float}
 MainFire.Random.MaxVolume {type=Float}
 MainFire.Random.MinPitch {type=Float}
 MainFire.Random.MinVolume {type=Float}
 MainFire.Sounds {type=WaveAsset:Array}
 MainFire.StopTriggerDelay {type=Time}
 Mixing.LowAmmoVolume {type=Float}
 Mixing.MainFireVolume {type=Float}
 Mixing.WeaponOutputVolume {type=Float}
 OnStop {type=Trigger}
 ReTrigger {type=Trigger}
 UE.Source.OnPlay {type=Trigger}
 Weapon.AmmoRemaining {type=Int32}
 Weapon.ClipSize {type=Int32}
 Weapon.FireRate {type=Float}
 Weapon.FireStartTime {type=Float}
 Weapon.IsAimingDownSights {type=Bool}
 Weapon.IsSuppressorEnabled {type=Bool}
]
interfaces [Fort.Character:1.0, Fort.Weapon:1.0, UE.OutputFormat.5dot1:1.0, UE.Source.OneShot:1.0, UE.Source:1.1]
outputs [
 bOnAlmostFinished {type=Bool}
 OnGunshotOut {type=Trigger}
 "UE.OutputFormat.5dot1.Audio:0" {type=Audio}
 "UE.OutputFormat.5dot1.Audio:1" {type=Audio}
 "UE.OutputFormat.5dot1.Audio:2" {type=Audio}
 "UE.OutputFormat.5dot1.Audio:3" {type=Audio}
 "UE.OutputFormat.5dot1.Audio:4" {type=Audio}
 "UE.OutputFormat.5dot1.Audio:5" {type=Audio}
 UE.Source.OneShot.OnFinished {type=Trigger}
]
pages [Default {input_defaults {
   Character.Affiliation=0
   Character.Health=0
   Character.UniqueID=0
   LowAmmo.ADSMainVolumeMultiply=1
   LowAmmo.Random.MaxPitch=1.05
   LowAmmo.Random.MaxVolume=1.05
   LowAmmo.Random.MinPitch=0.95
   LowAmmo.Random.MinVolume=0.95
   LowAmmo.Threshold=10
   MainFire.ADSMainVolumeMultiply=1
   MainFire.LFVolume=1
   MainFire.Random.MaxPitch=1
   MainFire.Random.MaxVolume=1.05
   MainFire.Random.MinPitch=0.9
   MainFire.Random.MinVolume=0.95
   MainFire.StopTriggerDelay=0.03
   Mixing.LowAmmoVolume=1
   Mixing.MainFireVolume=1
   Mixing.WeaponOutputVolume=0.9
   OnStop=null
   ReTrigger=null
   UE.Source.OnPlay=null
   Weapon.AmmoRemaining=30
   Weapon.ClipSize=30
   Weapon.FireRate=0.5
   Weapon.FireStartTime=0
   Weapon.IsAimingDownSights=null
   Weapon.IsSuppressorEnabled=null
   bAutoFire=null
   LowAmmo.Sounds []
   MainFire.Sounds [
    &/Hightower_Tomato_Primary_Fire_1P_Close_01
    &/Hightower_Tomato_Primary_Fire_1P_Close_02
    &/Hightower_Tomato_Primary_Fire_1P_Close_03
    &/Hightower_Tomato_Primary_Fire_1P_Close_04
    &/Hightower_Tomato_Primary_Fire_1P_Close_05
    &/Hightower_Tomato_Primary_Fire_1P_Close_06
   ]
  }}]
properties {
 assetImportData=&self:AssetImportData
 numChannels=6
 outputFormat=FiveDotOne
 qualitySetting=Fort_Weapons
 sampleRate=48000
 soundClassObject=&Fort_WeaponRanged_Fire_1P_PSM
 concurrencyOverrides {maxCount=2, resolutionRule=StopOldest}
 concurrencySet [&Weapon_Gunfire_Local_Conc]
}
