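# MetaSound preset for the third-person primary fire sound of the HighTower Tomato weapon, outputting stereo.
# It inherits its graph from MSS_Base_Weapon_Fire_3p_Simple and supplies close, medium and distant variation arrays, with the close layer using 1P recordings.
# Inputs set per-layer gains, distance curves and falloff, plus random pitch, volume and EQ ranges.
# Weapon attenuation and local and global gunfire concurrency sets are assigned.
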MetaSoundSource /HighTower/Audio/Tomato/MSS_Hightower_Tomato_Primary_Fire_3p_Preset
preset_parent=/FortMetaSoundWeapon/Items/Metasounds/MSS_Base_Weapon_Fire_3p_Simple
inputs [
 bAutoFire {type=Bool}
 Character.Affiliation {type=Enum:FortAudioAffiliation}
 Character.Health {type=Float}
 Character.UniqueID {type=Int32}
 MainFire.DistanceCurves {type=WaveTableBankAsset}
 MainFire.FalloffDistance {type=Float}
 MainFire.SoundsClose {type=WaveAsset:Array}
 MainFire.SoundsDistant {type=WaveAsset:Array}
 MainFire.SoundsMedium {type=WaveAsset:Array}
 MainFire.StopTriggerDelay {type=Time}
 Mixing.CloseGain {type=Float}
 Mixing.DistantGain {type=Float}
 Mixing.Friendly.FriendlyFireWeaponVolume {type=Float}
 Mixing.MediumGain {type=Float}
 Mixing.WeaponOutputVolume {type=Float}
 OnStop {type=Trigger}
 Random.EQ.A.MaxFreq {type=Float}
 Random.EQ.A.MaxGain {type=Float}
 Random.EQ.A.MinFreq {type=Float}
 Random.EQ.A.MinGain {type=Float}
 Random.EQ.B.MaxFreq {type=Float}
 Random.EQ.B.MaxGain {type=Float}
 Random.EQ.B.MinFreq {type=Float}
 Random.EQ.B.MinGain {type=Float}
 Random.MaxPitch {type=Float}
 Random.MaxVolume {type=Float}
 Random.MinPitch {type=Float}
 Random.MinVolume {type=Float}
 ReTrigger {type=Trigger}
 UE.Attenuation.Distance {type=Float}
 UE.Source.OnPlay {type=Trigger}
 UE.Source.Orientation.Azimuth {type=Float}
 UE.Source.Orientation.Elevation {type=Float}
 UE.Spatialization.Azimuth {type=Float}
 UE.Spatialization.Elevation {type=Float}
 Weapon.AmmoRemaining {type=Int32}
 Weapon.ClipSize {type=Int32}
 Weapon.FireRate {type=Float}
 Weapon.FireStartTime {type=Float}
 Weapon.IsAimingDownSights {type=Bool}
 Weapon.IsSuppressorEnabled {type=Bool}
]
interfaces [
 Fort.Character:1.0
 Fort.Weapon:1.0
 UE.Attenuation:1.0
 UE.OutputFormat.Stereo:1.0
 UE.Source.OneShot:1.0
 UE.Source.Orientation:1.0
 UE.Source:1.1
 UE.Spatialization:1.0
]
outputs [
 bOnAlmostFinished {type=Bool}
 CloseCurveEvaluation {type=Float}
 MediumCurveEvaluation {type=Float}
 OnGunshotOut {type=Trigger}
 "UE.OutputFormat.Stereo.Audio:0" {type=Audio}
 "UE.OutputFormat.Stereo.Audio:1" {type=Audio}
 UE.Source.OneShot.OnFinished {type=Trigger}
]
pages [Default {input_defaults {
   Character.Affiliation=0
   Character.Health=1
   Character.UniqueID=0
   MainFire.DistanceCurves=&WeaponRanged_DistanceCurves
   MainFire.FalloffDistance=27500
   MainFire.StopTriggerDelay=0.03
   Mixing.CloseGain=1
   Mixing.DistantGain=0.6
   Mixing.Friendly.FriendlyFireWeaponVolume=0.7
   Mixing.MediumGain=0.7
   Mixing.WeaponOutputVolume=0.8
   OnStop=null
   Random.EQ.A.MaxFreq=20000
   Random.EQ.A.MaxGain=0
   Random.EQ.A.MinFreq=699.9999
   Random.EQ.A.MinGain=-3
   Random.EQ.B.MaxFreq=8000
   Random.EQ.B.MaxGain=0
   Random.EQ.B.MinFreq=700
   Random.EQ.B.MinGain=-1
   Random.MaxPitch=1.15
   Random.MaxVolume=1.15
   Random.MinPitch=0.85
   Random.MinVolume=0.85
   ReTrigger=null
   UE.Attenuation.Distance=25000
   UE.Source.OnPlay=null
   UE.Source.Orientation.Azimuth=0
   UE.Source.Orientation.Elevation=0
   UE.Spatialization.Azimuth=0
   UE.Spatialization.Elevation=0
   Weapon.AmmoRemaining=30
   Weapon.ClipSize=30
   Weapon.FireRate=0.5
   Weapon.FireStartTime=0
   Weapon.IsAimingDownSights=null
   Weapon.IsSuppressorEnabled=null
   bAutoFire=null
   MainFire.SoundsClose [
    &/Hightower_Tomato_Primary_Fire_1P_Close_01
    &/Hightower_Tomato_Primary_Fire_1P_Close_02
    &/Hightower_Tomato_Primary_Fire_1P_Close_03
    &/Hightower_Tomato_Primary_Fire_1P_Close_04
    &/Hightower_Tomato_Primary_Fire_1P_Close_05
    &/Hightower_Tomato_Primary_Fire_1P_Close_06
   ]
   MainFire.SoundsDistant [
    &/Hightower_Tomato_Primary_Fire_3P_Distant_01
    &/Hightower_Tomato_Primary_Fire_3P_Distant_02
    &/Hightower_Tomato_Primary_Fire_3P_Distant_03
    &/Hightower_Tomato_Primary_Fire_3P_Distant_04
    &/Hightower_Tomato_Primary_Fire_3P_Distant_05
   ]
   MainFire.SoundsMedium [
    &/Hightower_Tomato_Primary_Fire_3P_Medium_01
    &/Hightower_Tomato_Primary_Fire_3P_Medium_02
    &/Hightower_Tomato_Primary_Fire_3P_Medium_03
    &/Hightower_Tomato_Primary_Fire_3P_Medium_04
    &/Hightower_Tomato_Primary_Fire_3P_Medium_05
    &/Hightower_Tomato_Primary_Fire_3P_Medium_06
   ]
  }}]
properties {
 assetImportData=&self:AssetImportData
 attenuationSettings=&Fort_Weapon_Projectile_NonAtt_Athena_Att
 maxDistance=27500
 numChannels=2
 outputFormat=Stereo
 qualitySetting=Fort_Weapons
 sampleRate=48000
 soundClassObject=&Fort_WeaponRanged_Fire_3P_PSM
 concurrencyOverrides {bLimitToOwner=true, maxCount=2, resolutionRule=StopOldest}
 concurrencySet [&Weapon_Gunfire_Local_Conc, &Weapon_Gunfire_Global_Conc]
}
