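# One-shot mono MetaSound that plays the overheat warning for the Hightower Tomato primary weapon.
# On play, it looks up a pitch shift in a float array variable using the ShotsLeft input as the index, so pitch can vary with remaining shots.
# A short trigger delay then starts a single wave, scaled by Master Volume, and the wave's finish ends the source.
# Concurrency is overridden to cap instances, stopping the oldest.
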MetaSoundSource /HighTower/Audio/Tomato/MSS_Hightower_Tomato_Primary_Overheat_Warning
inputs [
 Master Volume {type=Float}
 ShotsLeft {type=Int32}
 UE.Source.OnPlay {type=Trigger}
]
interfaces [UE.OutputFormat.Mono:1.0, UE.Source.OneShot:1.0, UE.Source:1.1]
outputs [
 "UE.OutputFormat.Mono.Audio:0" {type=Audio}
 UE.Source.OneShot.OnFinished {type=Trigger}
]
pages [Default {
  input_defaults {Master Volume=2, ShotsLeft=0, UE.Source.OnPlay=null}
  nodes [
   n1 = InitVariable.Float:Array()
   TriggerDelay_0 = UE.Trigger Delay(In=^UE.Source.OnPlay) {also {defaults {Delay Time=0.1}}}
   WavePlayer_0 = UE.Wave Player:Mono(
    Pitch Shift=Array.Get:Float:Array(Index=^ShotsLeft, Trigger=^UE.Source.OnPlay, Array=VariableAccessor.Float:Array(Variable=@n1.Variable).Value).Element
    Play=@TriggerDelay_0.Out
   ) {also {defaults {Wave Asset=&/Hightower_Tomato_Primary_Overheat_Warning_01}}}
   Multiply_0 = UE.Multiply:Audio by Float(AdditionalOperands=^Master Volume, PrimaryOperand=@WavePlayer_0.Out Mono)
  ]
  links [{dst="@UE.OutputFormat.Mono.Audio:0", src=Multiply_0.Out}, {dst="@UE.Source.OneShot.OnFinished", src=WavePlayer_0.On Finished}]
 }]
properties {
 assetImportData=&self:AssetImportData
 attenuationSettings=&Fort_GadgetActivate_Att
 bOverrideConcurrency=true
 maxDistance=4000
 sampleRate=48000
 soundClassObject=&Fort_UI_InGame_NonCritical
 volume=2
 concurrencyOverrides {maxCount=3, resolutionRule=StopOldest}
}
