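# First-person primary fire sound cue for the HighTower Tomato weapon, routed to the Fortnite ranged-weapon 1P fire sound class.
# A Random node picks one of five equally weighted close-perspective fire waves.
# A root Modulator applies a fixed upward pitch shift at unchanged volume, so variation comes only from sample choice.
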SoundCue /HighTower/Audio/Tomato/SC_Hightower_Tomato_Primary_Fire_1p
root=Modulator_0
nodes [
 {id=Modulator_0, children [Random_1], properties {pitchMax=1.1, pitchMin=1.1, volumeMax=1, volumeMin=1}}
 {id=Random_1, children [WavePlayer_0,WavePlayer_1,WavePlayer_2,WavePlayer_3,WavePlayer_6], properties {weights [1,1,1,1,1]}}
 {id=WavePlayer_0, properties {soundWaveAssetPtr=&/Hightower_Tomato_Primary_Fire_1P_Close_02}}
 {id=WavePlayer_1, properties {soundWaveAssetPtr=&/Hightower_Tomato_Primary_Fire_1P_Close_03}}
 {id=WavePlayer_2, properties {soundWaveAssetPtr=&/Hightower_Tomato_Primary_Fire_1P_Close_04}}
 {id=WavePlayer_3, properties {soundWaveAssetPtr=&/Hightower_Tomato_Primary_Fire_1P_Close_05}}
 {id=WavePlayer_6, properties {soundWaveAssetPtr=&/Hightower_Tomato_Primary_Fire_1P_Close_06}}
]
properties {maxDistance=2097152, soundClassObject=&Fort_WeaponRanged_Fire_1P_PSM}
