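# Sound cue for the third-person primary fire of the Hightower Tomato weapon, on the third-person ranged-fire sound class with a long max distance.
# A mixer layers close, medium and distant branches, each with its own projectile attenuation, and lowers the medium layer's volume.
# Each branch picks randomly among equally weighted waves through a modulator.
# The close layer reuses first-person close samples at a fixed raised pitch.
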SoundCue /HighTower/Audio/Tomato/SC_Hightower_Tomato_Primary_Fire_3p
root=Mixer_2
nodes [
 {id=Attenuation_3, children [Modulator_0], properties {attenuationSettings=&Fort_Weapon_Projectile_Close_Athena_Att}}
 {
  id=Attenuation_4
  children [Modulator_4]
  properties {attenuationSettings=&Fort_Weapon_Projectile_Distant_Mono_Athena_Att}
 }
 {
  id=Attenuation_5
  children [Modulator_1]
  properties {attenuationSettings=&Fort_Weapon_Projectile_Medium_Mono_Athena_Att}
 }
 {id=Mixer_2, children [Attenuation_3, Attenuation_5, Attenuation_4], properties {inputVolume [1,0.4,1]}}
 {id=Modulator_0, children [Random_1], properties {pitchMax=1.1, pitchMin=1.1, volumeMax=1, volumeMin=1}}
 {id=Modulator_1, children [Random_0], properties {volumeMax=1, volumeMin=0.9}}
 {id=Modulator_4, children [Random_7]}
 {
  id=Random_0
  children [WavePlayer_17, WavePlayer_18, WavePlayer_19, WavePlayer_20, WavePlayer_21, WavePlayer_22]
  properties {weights [1,1,1,1,1,1]}
 }
 {id=Random_1, children [WavePlayer_0, WavePlayer_1, WavePlayer_2, WavePlayer_9, WavePlayer_10], properties {weights [1,1,1,1,1]}}
 {id=Random_7, children [WavePlayer_12, WavePlayer_13, WavePlayer_14, WavePlayer_15, WavePlayer_16], properties {weights [1,1,1,1,1]}}
 {id=WavePlayer_0, properties {soundWaveAssetPtr=&/Hightower_Tomato_Primary_Fire_1P_Close_02}}
 {id=WavePlayer_1, properties {soundWaveAssetPtr=&/Hightower_Tomato_Primary_Fire_1P_Close_03}}
 {id=WavePlayer_10, properties {soundWaveAssetPtr=&/Hightower_Tomato_Primary_Fire_1P_Close_06}}
 {id=WavePlayer_12, properties {soundWaveAssetPtr=&/Hightower_Tomato_Primary_Fire_3P_Distant_01}}
 {id=WavePlayer_13, properties {soundWaveAssetPtr=&/Hightower_Tomato_Primary_Fire_3P_Distant_02}}
 {id=WavePlayer_14, properties {soundWaveAssetPtr=&/Hightower_Tomato_Primary_Fire_3P_Distant_03}}
 {id=WavePlayer_15, properties {soundWaveAssetPtr=&/Hightower_Tomato_Primary_Fire_3P_Distant_04}}
 {id=WavePlayer_16, properties {soundWaveAssetPtr=&/Hightower_Tomato_Primary_Fire_3P_Distant_05}}
 {id=WavePlayer_17, properties {soundWaveAssetPtr=&/Hightower_Tomato_Primary_Fire_3P_Medium_01}}
 {id=WavePlayer_18, properties {soundWaveAssetPtr=&/Hightower_Tomato_Primary_Fire_3P_Medium_02}}
 {id=WavePlayer_19, properties {soundWaveAssetPtr=&/Hightower_Tomato_Primary_Fire_3P_Medium_03}}
 {id=WavePlayer_2, properties {soundWaveAssetPtr=&/Hightower_Tomato_Primary_Fire_1P_Close_04}}
 {id=WavePlayer_20, properties {soundWaveAssetPtr=&/Hightower_Tomato_Primary_Fire_3P_Medium_04}}
 {id=WavePlayer_21, properties {soundWaveAssetPtr=&/Hightower_Tomato_Primary_Fire_3P_Medium_05}}
 {id=WavePlayer_22, properties {soundWaveAssetPtr=&/Hightower_Tomato_Primary_Fire_3P_Medium_06}}
 {id=WavePlayer_9, properties {soundWaveAssetPtr=&/Hightower_Tomato_Primary_Fire_1P_Close_05}}
]
properties {maxDistance=27500, soundClassObject=&Fort_WeaponRanged_Fire_3P_PSM}
